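# Sound cue for the opening sound of the BoneWasp glider cosmetic.
# A single wave plays through a modulator with fixed reduced volume and no pitch variation, then into a mixer whose second input is empty.
# Shared glider open/close attenuation and concurrency, a glider-open audio property sheet, and the Fort_Gliders_Land sound class govern playback.
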SoundCue /BRCosmetics/Sounds/Gliders/BoneWasp/Glider_BoneWasp_Open_Cue
root=Mixer_0
nodes [
 {id=Mixer_0, children [Modulator_0,null], properties {inputVolume [1,0.35]}}
 {id=Modulator_0, children [WavePlayer_1], properties {pitchMax=1, pitchMin=1, volumeMax=0.6, volumeMin=0.6}}
 {id=WavePlayer_1, properties {soundWaveAssetPtr=&/Glider_BoneWasp_Open_01}}
]
properties {
 attenuationSettings=/Game/Sounds/Attenuation/Fort_Player_Glider_OpenClose_Athena_Att
 audioPropertiesSheet=/BRCosmetics/Sounds/Settings/PropertySheets/Glider/APS_Cosmetic_Glider_Open
 maxDistance=10000
 soundClassObject=&Fort_Gliders_Land
 volumeMultiplier=1
 concurrencySet [&Athena_GliderOpenClose_Conc]
}
